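# Control Rig for the F_MED_RedOasisApricot FaceAcc2 cosmetic part, adjusting hair bones as the head rotates.
# The forward solve runs two spherical pose readers on the head bone, with the same rotation offset but different active region scales.
# Each reader's output weights chained Modify Transforms: the first on back-center and right base hair, the second on two back-center bones.
# The second reader has debug drawing on, and the right-back transform variable is unconnected.
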ControlRigBlueprint /BRCosmetics/Characters/Player/Female/Medium/Bodies/F_MED_RedOasisApricot/Meshes/Parts/F_MED_RedOasisApricot_FaceAcc2_CtrlRig
graphs [
 ContainedGraph {nodes [
   n1 = RigVMFunctionEntryNode()
   SequenceAggregate = &RigVMFunction_Sequence(ExecuteContext=@n1.ExecuteContext)
   SequenceAggregate_1 = &RigVMFunction_Sequence(ExecuteContext=@SequenceAggregate.B)
   RigVMFunctionReturnNode(A=@SequenceAggregate.A, B=@SequenceAggregate_1.A, C=@SequenceAggregate_1.B)
  ]}
 RigVMFunctionLibrary {links [], nodes []}
 RigVMModel {nodes [
   RigUnit_BeginExecution = &RigUnit_BeginExecution()
   SequenceAggregate = RigVMAggregateNode(ExecuteContext=@RigUnit_BeginExecution.ExecutePin)
   n1 = &RigUnit_SphericalPoseReader(
    ActiveRegionScaleFactors=(PositiveWidth=1,NegativeWidth=1,PositiveHeight=0.5,NegativeHeight=1)
    ActiveRegionSize=0.1
    Debug=(bDrawDebug=false,bDraw2D=False,bDrawLocalAxes=False,DebugScale=25,DebugSegments=20,DebugThickness=0.25)
    DriverAxis=(X=1,Y=0,Z=0)
    DriverItem=(Type=Bone,Name="head")
    FalloffRegionScaleFactors=(PositiveWidth=1,NegativeWidth=1,PositiveHeight=1.8,NegativeHeight=1)
    FalloffSize=0.18
    FlipHeightScaling=false
    FlipWidthScaling=false
    OptionalParentItem=(Type=Bone)
    RotationOffset=(X=-34,Y=-90,Z=0)
    ExecutePin=@SequenceAggregate.A
   )
   > &RigUnit_ModifyTransforms(
    ItemToModify=((Item=(Type=Bone,Name="dyn_hair_C_back_01"),Transform=(Rotation=(X=0,Y=0,Z=0,W=1),Translation=(X=0,Y=0,Z=0),Scale3D=(X=1,Y=1,Z=1))))
    Weight=@n1.OutputParam
    ItemToModify.0.Transform=RigVMVariableNode() {variable=Hair C_Back_transform}.Value
   )
   > &RigUnit_ModifyTransforms(
    ItemToModify=((Item=(Type=Bone,Name="dyn_hair_C_back_01"),Transform=(Rotation=(X=0,Y=0,Z=0,W=1),Translation=(X=0,Y=0,Z=0),Scale3D=(X=1,Y=1,Z=1))))
    ItemToModify.0.Transform.Rotation=&RigVMFunction_MathQuaternionFromEuler(Euler=(X=-20,Y=0,Z=-40)).Result
    Weight=@n1.OutputParam
   )
   > &RigUnit_ModifyTransforms(
    ItemToModify=((Item=(Type=Bone,Name="dyn_hair_base_r"),Transform=(Rotation=(X=0,Y=0,Z=0,W=1),Translation=(X=0,Y=0,Z=0),Scale3D=(X=1,Y=1,Z=1))))
    ItemToModify.0.Transform.Rotation=&RigVMFunction_MathQuaternionFromEuler(Euler=(X=80,Y=60,Z=-60)).Result
    Weight=@n1.OutputParam
   )
   n2 = &RigUnit_SphericalPoseReader(
    ActiveRegionScaleFactors=(PositiveWidth=1.2,NegativeWidth=1.2,PositiveHeight=1.2,NegativeHeight=1.2)
    ActiveRegionSize=0.1
    Debug=(bDrawDebug=True,bDraw2D=False,bDrawLocalAxes=False,DebugScale=25,DebugSegments=20,DebugThickness=0.25)
    DriverAxis=(X=1,Y=0,Z=0)
    DriverItem=(Type=Bone,Name="head")
    FalloffRegionScaleFactors=(PositiveWidth=1,NegativeWidth=1,PositiveHeight=1.8,NegativeHeight=1)
    FalloffSize=0.18
    FlipHeightScaling=false
    FlipWidthScaling=false
    OptionalParentItem=(Type=Bone)
    RotationOffset=(X=-34,Y=-90,Z=0)
    ExecutePin=@SequenceAggregate.B
   )
   > &RigUnit_ModifyTransforms(
    ItemToModify=((Item=(Type=Bone,Name="dyn_hair_C_back_01"),Transform=(Rotation=(X=0,Y=0,Z=0,W=1),Translation=(X=0,Y=0,Z=0),Scale3D=(X=1,Y=1,Z=1))))
    ItemToModify.0.Transform.Rotation=&RigVMFunction_MathQuaternionFromEuler(Euler=(X=20,Y=40,Z=-20)).Result
    Weight=@n2.OutputParam
   )
   > &RigUnit_ModifyTransforms(
    ItemToModify=((Item=(Type=Bone,Name="dyn_hair_C_back_02"),Transform=(Rotation=(X=0,Y=0,Z=0,W=1),Translation=(X=0,Y=0,Z=0),Scale3D=(X=1,Y=1,Z=1))))
    ItemToModify.0.Transform.Rotation=&RigVMFunction_MathQuaternionFromEuler(Euler=(X=0,Y=0,Z=0)).Result
    Weight=@n2.OutputParam
   )
   &RigVMFunction_MathQuaternionFromEuler(Euler=(X=0,Y=30,Z=0))
   RigVMVariableNode() {variable=Hair R Back Transform}
  ]}
]
hierarchy [
 root {loc=<0 0 0>, rot=<0 0 0 1>}
 pelvis {parent=root, loc=<0 -0.5301 98.0571>, rot=<0.0127 0.707 -0.0127 -0.707>}
 spine_01 {parent=pelvis, loc=<6.1208 -0.9328 0>, rot=<0 0 0.034 -0.9994>}
 spine_02 {parent=spine_01, loc=<6.5996 -0.2095 0>, rot=<0 0 0.0238 -0.9997>}
 spine_03 {parent=spine_02, loc=<6.4914 -0.1237 0>, rot=<0 0 0.0149 -0.9999>}
 spine_04 {parent=spine_03, loc=<6.9598 0.0794 0>, rot=<0 0 -0.0132 -0.9999>}
 spine_05 {parent=spine_04, loc=<6.9987 0 0>, rot=<0 0 -0.0015 -1>}
 clavicle_l {parent=spine_05, loc=<5.4831 -0.1053 -3.2946>, rot=<0.0465 -0.7289 -0.009 -0.6829>}
 upperarm_l {parent=clavicle_l, loc=<9.8658 -1.2136 -0.4104>, rot=<0.0164 -0.3245 -0.0269 -0.9454>}
 neck_01 {parent=spine_05, loc=<11.3138 0.1529 0>, rot=<0 0 -0.2003 -0.9797>}
 neck_02 {parent=neck_01, loc=<4.5627 0 0>, rot=<0 0 0.0504 -0.9987>}
 head {parent=neck_02, loc=<5.3914 0 0>, rot=<0 0 0.115 -0.9934>}
 face_acc_offset {parent=head, loc=<2.7283 0.0226 0>, rot=<0 0 0 1>}
 dyn_beanie_base {parent=face_acc_offset, loc=<8.7114 -6.2346 0.2774>, rot=<-0.6567 0.2622 0.2622 -0.6567>}
 dyn_beanie {parent=dyn_beanie_base, loc=<0 0 0>, rot=<0 0 0 1>}
 dyn_strand_base_l {parent=face_acc_offset, loc=<-2.1256 5.1002 -6.884>, rot=<0.0963 -0.276 0.9055 0.3076>}
 dyn_strand_parent_l {parent=dyn_strand_base_l, loc=<0 0 0>, rot=<0 0 0 1>}
 dyn_strand_01_l {parent=dyn_strand_parent_l, loc=<3.0834 -0.128 -0.1188>, rot=<0 0 0.0698 -0.9976>}
 dyn_strand_02_l {parent=dyn_strand_01_l, loc=<3.6999 -0.1014 -0.5164>, rot=<-0.0186 -0.2121 -0.1654 -0.963>}
 dyn_hair_base_l {parent=face_acc_offset, loc=<-3.9265 -1.6103 -9.0112>, rot=<-0.2863 -0.6073 0.6983 0.2482>}
 dyn_hair_parent_l {parent=dyn_hair_base_l, loc=<0 0 0>, rot=<0 0 0 1>}
 dyn_hair_01_l {parent=dyn_hair_parent_l, loc=<9.2546 -0.562 2.1499>, rot=<0.141 0 0.141 -0.9799>}
 dyn_hair_02_l {parent=dyn_hair_01_l, loc=<5.3114 -0.7231 0.4885>, rot=<-0.1209 -0.0629 0.1922 -0.9718>}
 dyn_hair_03_l {parent=dyn_hair_02_l, loc=<5.1596 -0.3239 1.8261>, rot=<-0.0627 0.1784 -0.1729 -0.9666>}
 dyn_hair_04_l {parent=dyn_hair_03_l, loc=<3.7751 1.7938 0.8268>, rot=<0.0288 0.0133 -0.2078 -0.9777>}
 dyn_strand_base_r {parent=face_acc_offset, loc=<0.7763 6.8702 6.0744>, rot=<0.0228 -0.9429 0.2536 0.2146>}
 dyn_strand_parent_r {parent=dyn_strand_base_r, loc=<0 0 0>, rot=<0 0 0 1>}
 dyn_strand_01_r {parent=dyn_strand_parent_r, loc=<3.9327 -0.3634 -0.7311>, rot=<0.1396 -0.1991 0.1971 -0.9498>}
 dyn_strand_02_r {parent=dyn_strand_01_r, loc=<2.6017 -1.4143 -0.1316>, rot=<-0.0092 0.0565 0.569 -0.8204>}
 dyn_hair_base_r {parent=face_acc_offset, loc=<-1.8256 0.8526 7.9714>, rot=<0.004 -0.9563 0.0012 0.2924>}
 dyn_hair_parent_r {parent=dyn_hair_base_r, loc=<0 0 1.6385>, rot=<0 0 0 1>}
 dyn_hair_01_r {parent=dyn_hair_parent_r, loc=<4.5917 0.649 2.4698>, rot=<-0.0012 0.0282 -0.0822 -0.9962>}
 dyn_hair_02_r {parent=dyn_hair_01_r, loc=<3.2771 0.9523 1.3052>, rot=<0.0365 0.1368 -0.0629 -0.9879>}
 dyn_hair_03_r {parent=dyn_hair_02_r, loc=<4.5094 0.4513 0.5587>, rot=<-0.048 0.1174 -0.039 -0.9912>}
 dyn_hair_04_r {parent=dyn_hair_03_r, loc=<3.442 0.1527 0.44>, rot=<0.2518 0.0063 0.0535 -0.9663>}
 dyn_hair_05_r {parent=dyn_hair_04_r, loc=<3.0113 0.5725 0.9518>, rot=<0.0799 0.3864 -0.0487 -0.9176>}
 dyn_hair_11_r {parent=dyn_hair_parent_r, loc=<3.6154 -2.817 0.8419>, rot=<0.9867 0.0287 0.1268 0.0972>}
 dyn_hair_12_r {parent=dyn_hair_11_r, loc=<3.9105 0 0>, rot=<0.0303 -0.0881 0.2778 -0.9561>}
 dyn_hair_C_back_01 {parent=face_acc_offset, loc=<-1.7825 -5.8392 1.0081>, rot=<-0.2307 0.6684 0.6684 -0.2307>}
 dyn_hair_C_back_02 {parent=dyn_hair_C_back_01, loc=<2.1864 0 0>, rot=<0 0.153 0 -0.9882>}
 dyn_hair_C_back_03 {parent=dyn_hair_C_back_02, loc=<3.0592 0 0>, rot=<0 -0.0134 0 -0.9999>}
 dyn_hair_C_back_04 {parent=dyn_hair_C_back_03, loc=<4.4111 0 0>, rot=<0 0.1329 0 -0.9911>}
 clavicle_r {parent=spine_05, loc=<5.4831 -0.1053 3.2946>, rot=<0.7289 0.0465 -0.6829 0.009>}
 upperarm_r {parent=clavicle_r, loc=<-9.8658 1.2136 0.4104>, rot=<0.0164 -0.3245 -0.0269 -0.9454>}
 dyn_collar_l {parent=spine_05, loc=<7.1842 5.7605 -6.9864>, rot=<-0.0696 -0.3657 -0.3462 -0.8611>}
 dyn_collar_r {parent=spine_05, loc=<7.0911 6.7009 7.2245>, rot=<0.8651 0.289 0.4044 -0.0672>}
 dyn_turtle_r {parent=spine_05, loc=<16.1156 5.2964 7.0229>, rot=<0.7065 0.0284 0.7065 -0.0284>}
 dyn_turtle_l {parent=spine_05, loc=<16.1156 5.2964 -6.4905>, rot=<0.0284 -0.7065 -0.0284 -0.7065>}
]
variables [
 Hair C_Back_transform {type=FTransform, default=(Rotation=(X=-0.2241,Y=0.483,Z=0.1294,W=0.8365),Translation=(X=0,Y=0,Z=0),Scale3D=(X=1,Y=1,Z=1))}
 Hair R Back Transform {type=FTransform, default=(Rotation=(X=0.7011,Y=-0.0923,Z=-0.561,W=0.4305),Translation=(X=0,Y=0,Z=0),Scale3D=(X=1,Y=1,Z=1))}
]
