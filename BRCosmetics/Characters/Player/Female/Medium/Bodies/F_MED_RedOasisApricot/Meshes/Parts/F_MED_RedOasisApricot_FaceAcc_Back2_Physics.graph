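# Physics asset for the FaceAcc_Back2 part of the medium female RedOasisApricot character, covering its dynamic back and right-shoulder hair.
# A four-bone back chain with tight, nearly twist-free limits and a looser two-bone shoulder chain both hang from the head.
# Capsules on head, upper spine, clavicles and upper arms form the torso shapes, linked spine-to-arm by free-rotation constraints.
# The last two back-hair bones add long crosswise capsules for wider volume.
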PhysicsAsset /BRCosmetics/Characters/Player/Female/Medium/Bodies/F_MED_RedOasisApricot/Meshes/Parts/F_MED_RedOasisApricot_FaceAcc_Back2_Physics
bodies {
 clavicle_l [{shapeType=Capsule, length=10, radius=5, shapeName=clavicle_l_capsule, center=<5 0 -0.4719>, rotation=<rot 90 0 0>}]
 clavicle_r [{shapeType=Capsule, length=8.4317, radius=6.7683, shapeName=clavicle_l_capsule, center=<-5 0.2696 -0.1722>, rotation=<rot 90 0 0>}]
 dyn_hair_C_back_01 [{shapeType=Capsule, length=1, radius=1, shapeName=dyn_hair_C_back_01_capsule, rotation=<rot 90 0 0>}]
 dyn_hair_C_back_02 [{shapeType=Capsule, length=1, radius=1, shapeName=dyn_hair_C_back_02_capsule, center=<0.5 0 0>, rotation=<rot 90 0 0>}]
 dyn_hair_C_back_03 [
  {shapeType=Capsule, length=0, radius=3.25, shapeName=dyn_hair_C_back_04_capsule, center=<0.5 0 0>, rotation=<rot 90 0 0>}
  {shapeType=Capsule, length=15.4991, radius=3.05, shapeName=dyn_hair_C_back_04_capsule, center=<3.4512 0.3487 -0.1099>, rotation=<rot 0 0 90>}
 ]
 dyn_hair_C_back_04 [
  {shapeType=Capsule, length=1, radius=1, shapeName=dyn_hair_C_back_04_capsule, center=<0.5 0 0>, rotation=<rot 90 0 0>}
  {shapeType=Capsule, length=13.9991, radius=3.05, shapeName=dyn_hair_C_back_04_capsule, center=<3.4512 0.3487 -0.1099>, rotation=<rot 0 0 90>}
 ]
 dyn_rShoulder_hair_01 [{shapeType=Capsule, length=5.1239, radius=4.8761, shapeName=dyn_hair_C_back_04_capsule, center=<0 0 3>, rotation=<rot 0 0 90>}]
 dyn_rShoulder_hair_02 [{shapeType=Capsule, length=5.1239, radius=4.8761, shapeName=dyn_hair_C_back_04_capsule, center=<0 0 3>, rotation=<rot 0 0 90>}]
 head [{shapeType=Capsule, length=5, radius=5, shapeName=head_capsule, center=<-0.0733 2.5304 0>, rotation=<rot 90 0 0>}]
 spine_05 [
  {shapeType=Capsule, length=10.713, radius=6.037, shapeName=spine_05_capsule, center=<7.91 0.1471 0>, rotation=<rot 90 0 -19.9998>}
  {shapeType=Capsule, length=28.5838, radius=10.7877, shapeName=spine_05_capsule, center=<-5.6357 0.5359 0>}
 ]
 upperarm_l [{shapeType=Capsule, length=19.2152, radius=5.5522, shapeName=upperarm_l_capsule, center=<13.3502 0 0.2057>, rotation=<rot 90 0 0>}]
 upperarm_r [{shapeType=Capsule, length=19.9652, radius=4.8022, shapeName=upperarm_l_capsule, center=<-13.9278 0 0.5261>, rotation=<rot 90 0 0>}]
}
constraints [
 {
  bone1Name=upperarm_l
  bone2Name=clavicle_l
  swing1LimitDegrees=45
  swing1Motion=Free
  swing2LimitDegrees=45
  swing2Motion=Free
  twistLimitDegrees=45
  twistMotion=Free
 }
 {
  bone1Name=upperarm_r
  bone2Name=clavicle_r
  swing1LimitDegrees=45
  swing1Motion=Free
  swing2LimitDegrees=45
  swing2Motion=Free
  twistLimitDegrees=45
  twistMotion=Free
 }
 {
  bone1Name=clavicle_l
  bone2Name=spine_05
  swing1LimitDegrees=45
  swing1Motion=Free
  swing2LimitDegrees=45
  swing2Motion=Free
  twistLimitDegrees=45
  twistMotion=Free
 }
 {
  bone1Name=clavicle_r
  bone2Name=spine_05
  swing1LimitDegrees=45
  swing1Motion=Free
  swing2LimitDegrees=45
  swing2Motion=Free
  twistLimitDegrees=45
  twistMotion=Free
 }
 {bone1Name=dyn_hair_C_back_01, bone2Name=head, swing1LimitDegrees=5, swing2LimitDegrees=5, twistLimitDegrees=1}
 {bone1Name=dyn_hair_C_back_02, bone2Name=dyn_hair_C_back_01, swing1LimitDegrees=16, swing2LimitDegrees=30, twistLimitDegrees=1}
 {bone1Name=dyn_hair_C_back_03, bone2Name=dyn_hair_C_back_02, swing1LimitDegrees=16, swing2LimitDegrees=20, twistLimitDegrees=1}
 {bone1Name=dyn_hair_C_back_04, bone2Name=dyn_hair_C_back_03, swing1LimitDegrees=16, swing2LimitDegrees=16, twistLimitDegrees=1}
 {bone1Name=dyn_rShoulder_hair_01, bone2Name=head, swing1LimitDegrees=35, swing2LimitDegrees=35, twistLimitDegrees=20}
 {bone1Name=dyn_rShoulder_hair_02, bone2Name=dyn_rShoulder_hair_01, swing1LimitDegrees=35, swing2LimitDegrees=35, twistLimitDegrees=20}
]
