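# Physics asset for the RedOasisApricot female face accessory, named as the skydive variant, covering beanie, hair and strand bones.
# Left and right hair and strand chains, plus a short beanie link, use angular-limited constraints between mostly capsule bodies.
# Head, spine, clavicle and upper-arm capsules add upper-body shapes, linked by angularly free constraints.
# Chain bases attach freely to spine_04, and dyn_hair_01_r also links to spine_05.
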PhysicsAsset /BRCosmetics/Characters/Player/Female/Medium/Bodies/F_MED_RedOasisApricot/Meshes/Parts/F_MED_RedOasisApricot_FaceAcc_Skydive_Physics
bodies {
 clavicle_l [{
   shapeType=Capsule
   length=9.7375
   radius=6.5
   shapeName=clavicle_l_capsule
   center=<5.6067 0 -1.4944>
   rotation=<rot 88.9738 -179.9593 -179.9594>
  }]
 clavicle_r [{
   shapeType=Capsule
   length=9.7375
   radius=6.5
   shapeName=clavicle_l_capsule
   center=<-5.7342 1.3868 1.6168>
   rotation=<rot 88.9738 -179.9911 -179.9911>
  }]
 dyn_beanie [{shapeType=Capsule, length=1.4037, radius=2.3187, shapeName=dyn_beanie_capsule, center=<0.2712 0 0>, rotation=<rot 0 -90 90>}]
 dyn_beanie_base [{shapeType=Sphere, radius=0.5, shapeName=dyn_beanie_base_sphere}]
 dyn_hair_01_l [{shapeType=Capsule, length=5.7886, radius=2.75, shapeName=dyn_hair_01_l_capsule, center=<1.7399 0 -3>, rotation=<rot 90 0 0>}]
 dyn_hair_01_r [
  {shapeType=Capsule, length=1.4548, radius=2, shapeName=dyn_hair_01_r_capsule, center=<0 0 -3>, rotation=<rot 90 0 0>}
  {shapeType=Capsule, length=3, radius=5.25, shapeName=dyn_hair_01_r_capsule, center=<5.5814 4.6064 -3>, rotation=<rot 80 0.0014 0.001>}
 ]
 dyn_hair_02_l [{shapeType=Capsule, length=2.7722, radius=3.5, shapeName=dyn_hair_02_l_capsule, center=<2.1877 0 -3>, rotation=<rot 90 0 0>}]
 dyn_hair_02_r [{shapeType=Capsule, length=0.5782, radius=2, shapeName=dyn_hair_02_r_capsule, center=<0 0 -3>, rotation=<rot 90 0 0>}]
 dyn_hair_03_l [{
   shapeType=Capsule
   length=1.5582
   radius=3.75
   shapeName=dyn_hair_03_l_capsule
   center=<1.2457 -0.5001 0.7098>
   rotation=<rot 65.5283 180 180>
  }]
 dyn_hair_03_r [{shapeType=Capsule, length=0.8203, radius=2, shapeName=dyn_hair_03_r_capsule, center=<0 0 -3>, rotation=<rot 90 0 0>}]
 dyn_hair_04_l [{shapeType=Capsule, length=0.4141, radius=2.75, shapeName=dyn_hair_04_l_capsule, center=<1.709 0 0>, rotation=<rot 90 0 0>}]
 dyn_hair_04_r [{shapeType=Capsule, length=1.8658, radius=3.5, shapeName=dyn_hair_04_r_capsule, center=<0 0 -3>, rotation=<rot 90 0 0>}]
 dyn_hair_05_r [{shapeType=Capsule, length=1.2499, radius=3.5, shapeName=dyn_hair_05_r_capsule, center=<0 0 -3>, rotation=<rot 90 0 0>}]
 dyn_hair_11_r [{
   shapeType=Capsule
   length=10.0726
   radius=7.642
   shapeName=dyn_hair_12_r_capsule
   center=<-4.5483 5.8191 -1.972>
   rotation=<rot -43.955 -37.4569 -11.5182>
  }]
 dyn_hair_12_r [{
   shapeType=Capsule
   length=10.0726
   radius=7.642
   shapeName=dyn_hair_12_r_capsule
   center=<-4.5483 5.8191 -1.972>
   rotation=<rot -43.955 -37.4569 -11.5182>
  }]
 dyn_hair_base_l [{shapeType=Sphere, radius=0.5, shapeName=dyn_hair_base_l_sphere}]
 dyn_hair_base_r [{shapeType=Capsule, length=0.505, radius=0.505, shapeName=dyn_hair_base_r_capsule, rotation=<rot 0 -90 0>}]
 dyn_hair_parent_l [{shapeType=Capsule, length=2.1231, radius=2, shapeName=dyn_hair_parent_l_capsule, center=<2.71 0 -3>, rotation=<rot 90 0 0>}]
 dyn_strand_01_l [{shapeType=Capsule, length=1.9572, radius=1.2984, shapeName=dyn_strand_01_l_capsule, center=<1.6021 0 0>, rotation=<rot 90 0 0>}]
 dyn_strand_01_r [{
   shapeType=Capsule
   length=1.3567
   radius=0.8459
   shapeName=dyn_strand_01_r_capsule
   center=<1.2057 -0.296 0>
   rotation=<rot 90 180 -154.5443>
  }]
 dyn_strand_02_l [{shapeType=Capsule, length=1.3909, radius=1.2267, shapeName=dyn_strand_02_l_capsule, center=<1.5879 0 0>, rotation=<rot 90 0 0>}]
 dyn_strand_02_r [{shapeType=Capsule, length=1.3813, radius=0.6063, shapeName=dyn_strand_02_r_capsule, center=<0.8069 0 0>, rotation=<rot 90 0 0>}]
 dyn_strand_base_l [{shapeType=Sphere, radius=0.5, shapeName=dyn_strand_base_l_sphere}]
 dyn_strand_base_r [{shapeType=Sphere, radius=0.5, shapeName=dyn_strand_base_r_sphere}]
 dyn_strand_parent_l [{shapeType=Capsule, length=0.7819, radius=1.4337, shapeName=dyn_strand_parent_l_capsule, center=<1.1275 0 0>, rotation=<rot 90 0 0>}]
 dyn_strand_parent_r [{
   shapeType=Capsule
   length=2.604
   radius=0.8421
   shapeName=dyn_strand_parent_r_capsule
   center=<1.7634 0 -0.2162>
   rotation=<rot 77.6927 0.0002 0.0002>
  }]
 head [{shapeType=Capsule, length=5.6022, radius=5.9913, shapeName=head_capsule, center=<-0.0733 2.5304 0>, rotation=<rot 90 0 0>}]
 spine_04 [
  {shapeType=Capsule, length=14.2273, radius=8.7837, shapeName=spine_05_capsule}
  {shapeType=Capsule, length=10.713, radius=6.037, shapeName=spine_05_capsule, center=<6.9312 2.8364 0>, rotation=<rot 90 -172.875 167.1252>}
 ]
 spine_05 [
  {shapeType=Capsule, length=14.2273, radius=8.7837, shapeName=spine_05_capsule}
  {shapeType=Capsule, length=10.713, radius=6.037, shapeName=spine_05_capsule, center=<6.9312 2.8364 0>, rotation=<rot 90 -172.875 167.1252>}
 ]
 upperarm_l [{shapeType=Capsule, length=19.2152, radius=7, shapeName=upperarm_l_capsule, center=<11.8069 0 0>, rotation=<rot 90 0 0>}]
 upperarm_r [{shapeType=Capsule, length=19.2152, radius=5, shapeName=upperarm_l_capsule, center=<-11.8069 0 0>, rotation=<rot 90 0 0>}]
}
constraints [
 {bone1Name=dyn_beanie, bone2Name=dyn_beanie_base, swing1LimitDegrees=5, swing2LimitDegrees=15, twistLimitDegrees=2}
 {bone1Name=dyn_strand_parent_l, bone2Name=dyn_strand_base_l, swing1LimitDegrees=10, swing2LimitDegrees=10, twistLimitDegrees=3}
 {bone1Name=dyn_strand_01_l, bone2Name=dyn_strand_parent_l, swing1LimitDegrees=10, swing2LimitDegrees=10, twistLimitDegrees=3}
 {bone1Name=dyn_strand_02_l, bone2Name=dyn_strand_01_l, swing1LimitDegrees=10, swing2LimitDegrees=10, twistLimitDegrees=3}
 {bone1Name=dyn_hair_parent_l, bone2Name=dyn_hair_base_l, swing1LimitDegrees=35, swing2LimitDegrees=35, twistLimitDegrees=5}
 {bone1Name=dyn_hair_01_l, bone2Name=dyn_hair_parent_l, swing1LimitDegrees=35, swing2LimitDegrees=35, twistLimitDegrees=5}
 {bone1Name=dyn_hair_02_l, bone2Name=dyn_hair_01_l, swing1LimitDegrees=35, swing2LimitDegrees=35, twistLimitDegrees=5}
 {bone1Name=dyn_hair_03_l, bone2Name=dyn_hair_02_l, swing1LimitDegrees=35, swing2LimitDegrees=35, twistLimitDegrees=5}
 {bone1Name=dyn_hair_04_l, bone2Name=dyn_hair_03_l, swing1LimitDegrees=35, swing2LimitDegrees=35, twistLimitDegrees=5}
 {bone1Name=dyn_strand_parent_r, bone2Name=dyn_strand_base_r, swing1LimitDegrees=5, swing2LimitDegrees=5, twistLimitDegrees=3}
 {bone1Name=dyn_strand_01_r, bone2Name=dyn_strand_parent_r, swing1LimitDegrees=25, swing2LimitDegrees=25, twistLimitDegrees=3}
 {bone1Name=dyn_strand_02_r, bone2Name=dyn_strand_01_r, swing1LimitDegrees=25, swing2LimitDegrees=25, twistLimitDegrees=3}
 {bone1Name=dyn_hair_01_r, bone2Name=dyn_hair_base_r, swing1LimitDegrees=35, swing2LimitDegrees=35, twistLimitDegrees=5}
 {bone1Name=dyn_hair_02_r, bone2Name=dyn_hair_01_r, swing1LimitDegrees=35, swing2LimitDegrees=35, twistLimitDegrees=5}
 {bone1Name=dyn_hair_03_r, bone2Name=dyn_hair_02_r, swing1LimitDegrees=35, swing2LimitDegrees=35, twistLimitDegrees=5}
 {bone1Name=dyn_hair_04_r, bone2Name=dyn_hair_03_r, swing1LimitDegrees=35, swing2LimitDegrees=35, twistLimitDegrees=5}
 {bone1Name=dyn_hair_05_r, bone2Name=dyn_hair_04_r, swing1LimitDegrees=35, swing2LimitDegrees=35, twistLimitDegrees=5}
 {
  bone1Name=upperarm_l
  bone2Name=clavicle_l
  swing1LimitDegrees=45
  swing1Motion=Free
  swing2LimitDegrees=45
  swing2Motion=Free
  twistLimitDegrees=45
  twistMotion=Free
 }
 {
  bone1Name=upperarm_r
  bone2Name=clavicle_r
  swing1LimitDegrees=45
  swing1Motion=Free
  swing2LimitDegrees=45
  swing2Motion=Free
  twistLimitDegrees=45
  twistMotion=Free
 }
 {
  bone1Name=clavicle_l
  bone2Name=spine_05
  swing1LimitDegrees=45
  swing1Motion=Free
  swing2LimitDegrees=45
  swing2Motion=Free
  twistLimitDegrees=45
  twistMotion=Free
 }
 {
  bone1Name=clavicle_r
  bone2Name=spine_05
  swing1LimitDegrees=45
  swing1Motion=Free
  swing2LimitDegrees=45
  swing2Motion=Free
  twistLimitDegrees=45
  twistMotion=Free
 }
 {bone1Name=dyn_hair_11_r, bone2Name=dyn_hair_base_r, swing1LimitDegrees=45, swing2LimitDegrees=45, twistLimitDegrees=15}
 {bone1Name=dyn_hair_12_r, bone2Name=dyn_hair_11_r, swing1LimitDegrees=25, swing2LimitDegrees=25, twistLimitDegrees=1}
 {
  bone1Name=spine_05
  bone2Name=spine_04
  swing1LimitDegrees=45
  swing1Motion=Free
  swing2LimitDegrees=45
  swing2Motion=Free
  twistLimitDegrees=45
  twistMotion=Free
 }
 {
  bone1Name=head
  bone2Name=spine_04
  swing1LimitDegrees=45
  swing1Motion=Free
  swing2LimitDegrees=45
  swing2Motion=Free
  twistLimitDegrees=45
  twistMotion=Free
 }
 {
  bone1Name=dyn_beanie_base
  bone2Name=spine_04
  swing1LimitDegrees=45
  swing1Motion=Free
  swing2LimitDegrees=45
  swing2Motion=Free
  twistLimitDegrees=45
  twistMotion=Free
 }
 {
  bone1Name=dyn_strand_base_l
  bone2Name=spine_04
  swing1LimitDegrees=45
  swing1Motion=Free
  swing2LimitDegrees=45
  swing2Motion=Free
  twistLimitDegrees=45
  twistMotion=Free
 }
 {
  bone1Name=dyn_hair_base_l
  bone2Name=spine_04
  swing1LimitDegrees=45
  swing1Motion=Free
  swing2LimitDegrees=45
  swing2Motion=Free
  twistLimitDegrees=45
  twistMotion=Free
 }
 {
  bone1Name=dyn_strand_base_r
  bone2Name=spine_04
  swing1LimitDegrees=45
  swing1Motion=Free
  swing2LimitDegrees=45
  swing2Motion=Free
  twistLimitDegrees=45
  twistMotion=Free
 }
 {
  bone1Name=dyn_hair_base_r
  bone2Name=spine_04
  swing1LimitDegrees=45
  swing1Motion=Free
  swing2LimitDegrees=45
  swing2Motion=Free
  twistLimitDegrees=45
  twistMotion=Free
 }
 {
  bone1Name=dyn_hair_01_r
  bone2Name=spine_05
  swing1LimitDegrees=45
  swing1Motion=Free
  swing2LimitDegrees=45
  swing2Motion=Free
  twistLimitDegrees=45
  twistMotion=Free
 }
]
