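# Animation Blueprint for the RedOasisApricot female FaceAcc cosmetic part, adding secondary motion to its hair and strand bones.
# A cached pose layers Body_Mesh arms onto the parent pose through a Control Rig, feeding two RigidBody chains (standard and skydive assets) with bone corrections.
# Simulation weight drops during melee and skydiving, with gravity partly driven by a state-based override.
# bRigidBodyEnabled switches between simulation and a Trail controller fallback.
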AnimBlueprint /BRCosmetics/Characters/Player/Female/Medium/Bodies/F_MED_RedOasisApricot/Meshes/Parts/F_MED_RedOasisApricot_FaceAcc_AnimBP
target_skeleton=&/F_MED_RedOasisApricot_FaceAcc_Skeleton
graphs [
 AnimGraph {nodes [
   Copy Pose From Mesh() {bUseAttachedParent=true}
   n1 = SelectFloat(A=0.4, B=SelectFloat(A=0.6, B=1, bPickA=Get bIsSkyDiving()), bPickA=Get bIsPlayingMeleeAnim())
   n2 = RigidBody!(
    ComponentPose=Local To Component(LocalPose=Use cached pose 'preRBAN'())
    Alpha=@n1
    OverrideWorldGravity:Vector=Get Gravity Override()
    ExternalForce
   ) -> (Pose) {
    bOverrideWorldGravity=true
    overridePhysicsAsset=/BRCosmetics/Characters/Player/Female/Medium/Bodies/F_MED_RedOasisApricot/Meshes/Icon/F_MED_RedOasisApricotIcon_FaceAcc_Physics
    simulationSpace=BaseBoneSpace
    alphaBoolBlend {blendOption=Linear}
    alphaScaleBiasClamp {bInterpResult=true}
    baseBoneRef {boneName=spine_05}
    componentLinearAccScale=<0.3 0.3 0.3>
    componentLinearVelScale=<0.3 0.3 0.3>
    overrideWorldGravity {z=-980}
   }
   n3 = RigidBody(ComponentPose=@n2, Alpha=@n1, OverrideWorldGravity="0,0,-980") {
    bOverrideWorldGravity=true
    overridePhysicsAsset=&/F_MED_RedOasisApricot_FaceAcc_Back2_Physics
    simulationSpace=BaseBoneSpace
    alphaBoolBlend {blendOption=Linear}
    alphaScaleBiasClamp {bInterpResult=true}
    baseBoneRef {boneName=spine_05}
    componentLinearAccScale=<0.3 0.3 0.3>
    componentLinearVelScale=<0.3 0.3 0.3>
    overrideWorldGravity {z=-980}
   }
   n4 = "Constraint - dyn_rShoulder_hair_01 (P)"(ComponentPose=@n3, Alpha=1, ConstraintWeights_0=1) {
    alphaBoolBlend {blendOption=Linear}
    boneToModify {boneName=dyn_rShoulder_hair_01}
    constraintSetup [{targetBone {boneName=clavicle_r}}]
    constraintWeights [1]
   }
   n5 = "Transform (Modify) Bone - Bone: dyn_rShoulder_hair_01"!(ComponentPose=@n4, bAlphaBoolEnabled:bool=True, Translation="0,0,0", Rotation="30,0,0", Scale="1,1,1") -> (Pose) {
    alphaInputType=Bool
    rotationMode=BMM_Additive
    rotationSpace=BCS_BoneSpace
    scaleMode=BMM_Additive
    scaleSpace=BCS_BoneSpace
    translationMode=BMM_Additive
    translationSpace=BCS_BoneSpace
    alphaBoolBlend {blendInTime=0.35, blendOption=Linear, blendOutTime=0.1}
    boneToModify {boneName=dyn_rShoulder_hair_01}
    rotation {pitch=30}
   }
   n6 = "Transform (Modify) Bone - Bone: dyn_hair_C_back_02"!(ComponentPose=@n5, bAlphaBoolEnabled:bool=True, Translation="0,0,0", Rotation="-15,0,0", Scale="1,1,1") -> (Pose) {
    alphaInputType=Bool
    rotationMode=BMM_Additive
    rotationSpace=BCS_BoneSpace
    scaleMode=BMM_Additive
    scaleSpace=BCS_BoneSpace
    translationMode=BMM_Additive
    translationSpace=BCS_BoneSpace
    alphaBoolBlend {blendInTime=0.6, blendOption=Linear, blendOutTime=0.1}
    boneToModify {boneName=dyn_hair_C_back_02}
    rotation {pitch=-15}
   }
   n7 = Component To Local(ComponentPose=@n6)
   n8 = SelectFloat(A=0.4, B=SelectFloat(A=0.6, B=1, bPickA=Get bIsSkyDiving()), bPickA=Get bIsPlayingMeleeAnim())
   n9 = RigidBody(ComponentPose=Local To Component(LocalPose=Use cached pose 'preRBAN'()), Alpha=@n8, OverrideWorldGravity=Get Gravity Override()) {
    bOverrideWorldGravity=true
    overridePhysicsAsset=&/F_MED_RedOasisApricot_FaceAcc_Skydive_Physics
    simulationSpace=BaseBoneSpace
    alphaBoolBlend {blendOption=Linear}
    alphaScaleBiasClamp {bInterpResult=true}
    baseBoneRef {boneName=spine_05}
    componentLinearAccScale=<0.3 0.3 0.3>
    componentLinearVelScale=<0.3 0.3 0.3>
    overrideWorldGravity {z=-980}
   }
   n10 = RigidBody(ComponentPose=@n9, Alpha=@n8, OverrideWorldGravity="0,0,-980") {
    bOverrideWorldGravity=true
    overridePhysicsAsset=&F_MED_RedOasisApricot_FaceAcc_Back2_Physics
    simulationSpace=BaseBoneSpace
    alphaBoolBlend {blendOption=Linear}
    alphaScaleBiasClamp {bInterpResult=true}
    baseBoneRef {boneName=spine_05}
    componentLinearAccScale=<0.3 0.3 0.3>
    componentLinearVelScale=<0.3 0.3 0.3>
    overrideWorldGravity {z=-980}
   }
   n11 = "Constraint - dyn_rShoulder_hair_01 (P)"(ComponentPose=@n10, Alpha=1, ConstraintWeights_0=1) {
    alphaBoolBlend {blendOption=Linear}
    boneToModify {boneName=dyn_rShoulder_hair_01}
    constraintSetup [{targetBone {boneName=clavicle_r}}]
    constraintWeights [1]
   }
   n12 = "Transform (Modify) Bone - Bone: dyn_rShoulder_hair_01"(ComponentPose=@n11, bAlphaBoolEnabled=True, Translation="0,0,0", Rotation="30,0,0", Scale="1,1,1") {
    alphaInputType=Bool
    rotationMode=BMM_Additive
    rotationSpace=BCS_BoneSpace
    scaleMode=BMM_Additive
    scaleSpace=BCS_BoneSpace
    translationMode=BMM_Additive
    translationSpace=BCS_BoneSpace
    alphaBoolBlend {blendInTime=0.35, blendOption=Linear, blendOutTime=0.1}
    boneToModify {boneName=dyn_rShoulder_hair_01}
    rotation {pitch=30}
   }
   n13 = "Transform (Modify) Bone - Bone: dyn_hair_C_back_02"(ComponentPose=@n12, bAlphaBoolEnabled=True, Translation="0,0,0", Rotation="-15,0,0", Scale="1,1,1") {
    alphaInputType=Bool
    rotationMode=BMM_Additive
    rotationSpace=BCS_BoneSpace
    scaleMode=BMM_Additive
    scaleSpace=BCS_BoneSpace
    translationMode=BMM_Additive
    translationSpace=BCS_BoneSpace
    alphaBoolBlend {blendInTime=0.6, blendOption=Linear, blendOutTime=0.1}
    boneToModify {boneName=dyn_hair_C_back_02}
    rotation {pitch=-15}
   }
   n14 = "Transform (Modify) Bone - Bone: dyn_rShoulder_hair_01"(ComponentPose=@n13, bAlphaBoolEnabled=True, Translation="0,0,-4", Rotation="0,0,0", Scale="1,1,1") {
    alphaInputType=Bool
    rotationMode=BMM_Additive
    rotationSpace=BCS_BoneSpace
    scaleMode=BMM_Additive
    scaleSpace=BCS_BoneSpace
    translationMode=BMM_Additive
    translationSpace=BCS_BoneSpace
    alphaBoolBlend {blendInTime=0.6, blendOption=Linear, blendOutTime=0.1}
    boneToModify {boneName=dyn_rShoulder_hair_01}
    translation {z=-4}
   }
   n15 = Component To Local(ComponentPose=@n14)
   n16 = "Constraint - dyn_hair_02_r (P)"(
    ComponentPose=Local To Component(LocalPose=Copy Pose From Mesh() {bCopyCurves=true, bUseAttachedParent=true})
    Alpha=1
    ConstraintWeights_0=1
   ) {
    alphaBoolBlend {blendOption=Linear}
    boneToModify {boneName=dyn_hair_02_r}
    constraintSetup [{targetBone {boneName=spine_05}}]
    constraintWeights [1]
   }
   n17 = "Constraint - dyn_hair_02_l (P)"(ComponentPose=@n16, Alpha=1, ConstraintWeights_0=1) {
    alphaBoolBlend {blendOption=Linear}
    boneToModify {boneName=dyn_hair_02_l}
    constraintSetup [{targetBone {boneName=spine_05}}]
    constraintWeights [1]
   }
   n18 = "Constraint - dyn_hair_C_back_01 (P)"(ComponentPose=@n17, Alpha=1, ConstraintWeights_0=1) {
    alphaBoolBlend {blendOption=Linear}
    boneToModify {boneName=dyn_hair_C_back_01}
    constraintSetup [{targetBone {boneName=spine_05}}]
    constraintWeights [1]
   }
   n19 = "Constraint - dyn_rShoulder_hair_01 (P)"(ComponentPose=@n18, Alpha=1, ConstraintWeights_0=1) {
    alphaBoolBlend {blendOption=Linear}
    boneToModify {boneName=dyn_rShoulder_hair_01}
    constraintSetup [{targetBone {boneName=spine_05}}]
    constraintWeights [1]
   }
   n20 = Trail controller - Bone: dyn_hair_04_l(ComponentPose=@n19, Alpha=1) {
    bLimitRotation=true
    bUsePlanarLimit=true
    chainLength=3
    planarLimits [{drivingBone {boneName=spine_05}, planeTransform {location=<0 10 0>, rotation=<rot 0 0 90>, scale=<1 1 1>}}]
    rotationLimits [{limitMax=<20 30 20>, limitMin=<-10 0 -10>}, {limitMax=<20 30 20>, limitMin=<-10 0 -10>}, {limitMax=<20 30 20>, limitMin=<-10 0 -10>}]
    rotationOffsets [<0 0 0>, <0 0 0>, <0 0 0>]
    trailBone {boneName=dyn_hair_04_l}
    trailRelaxationSpeed {editorCurveData {keys [<at 0 10 RCIM_Linear>,<at 0.9562 6.9898 RCIM_Linear>]}}
   }
   n21 = Trail controller - Bone: dyn_hair_05_r(ComponentPose=@n20, Alpha=1) {
    bLimitRotation=true
    bUsePlanarLimit=true
    chainLength=3
    planarLimits [{drivingBone {boneName=spine_05}, planeTransform {location=<0 9 0>, rotation=<rot 0 0 90>, scale=<1 1 1>}}]
    rotationLimits [{limitMax=<20 30 20>, limitMin=<-10 0 -10>}, {limitMax=<20 30 20>, limitMin=<-10 0 -10>}, {limitMax=<20 30 20>, limitMin=<-10 0 -10>}]
    rotationOffsets [<0 0 0>, <0 0 0>, <0 0 0>]
    trailBone {boneName=dyn_hair_05_r}
    trailRelaxationSpeed {editorCurveData {keys [<at -0.015 6.6299 RCIM_Linear>,<at 0.9595 6.0708 RCIM_Linear>]}}
   }
   n22 = Trail controller - Bone: dyn_hair_C_back_04(ComponentPose=@n21, Alpha=0.5) {
    bLimitRotation=true
    bUsePlanarLimit=true
    chainLength=3
    planarLimits [{drivingBone {boneName=spine_05}, planeTransform {location=<0 -10 0>, rotation=<rot 0 20 -90>, scale=<1 1 1>}}]
    rotationLimits [{limitMax=<30 30 30>, limitMin=<-30 -30 -30>}, {limitMax=<30 30 30>, limitMin=<-30 -30 -30>}, {limitMax=<30 30 30>, limitMin=<-30 -30 -30>}]
    rotationOffsets [<0 0 0>, <0 0 0>, <0 0 0>]
    trailBone {boneName=dyn_hair_C_back_04}
    trailRelaxationSpeed {editorCurveData {keys [<at -0.015 6.6299 RCIM_Linear>,<at 0.9505 6.3097 RCIM_Linear>]}}
   }
   n23 = Trail controller - Bone: dyn_strand_02_r(ComponentPose=@n22, Alpha=0.5) {
    bLimitRotation=true
    planarLimits [{drivingBone {boneName=spine_05}, planeTransform {location=<0 -10 0>, rotation=<rot 0 20 -90>, scale=<1 1 1>}}]
    rotationLimits [{limitMax=<30 30 30>, limitMin=<-30 -30 -30>}, {limitMax=<30 30 30>, limitMin=<-30 -30 -30>}]
    rotationOffsets [<0 0 0>, <0 0 0>]
    trailBone {boneName=dyn_strand_02_r}
    trailRelaxationSpeed {editorCurveData {keys [<at -0.015 6.6299 RCIM_Linear>,<at 0.9488 6.3203 RCIM_Linear>]}}
   }
   n24 = Trail controller - Bone: dyn_strand_02_l(ComponentPose=@n23, Alpha=0.5) {
    bLimitRotation=true
    planarLimits [{drivingBone {boneName=spine_05}, planeTransform {location=<0 -10 0>, rotation=<rot 0 20 -90>, scale=<1 1 1>}}]
    rotationLimits [{limitMax=<30 30 30>, limitMin=<-30 -30 -30>}, {limitMax=<30 30 30>, limitMin=<-30 -30 -30>}]
    rotationOffsets [<0 0 0>, <0 0 0>]
    trailBone {boneName=dyn_strand_02_l}
    trailRelaxationSpeed {editorCurveData {keys [<at -0.015 6.6299 RCIM_Linear>,<at 0.9488 6.3203 RCIM_Linear>]}}
   }
   Output Pose(Result=Blend Poses by bool(
     BlendPose_0=Blend Poses by bool(BlendPose_0=@n15, BlendPose_1=@n7) {blendType=Linear}
     BlendPose_1=Component To Local(ComponentPose=@n24)
     bActiveValue=Get bRigidBodyEnabled()
    ) {blendType=Linear}) {layerGroup=None}
   n25 = Get Body_Mesh()
   n26 = Control Rig!(
    Source=Layered blend per bone(BasePose=Copy Pose From Mesh() {bCopyCurves=true, bUseAttachedParent=true}, BlendPoses_0=Copy Pose From Mesh(SourceMeshComponent=@n25)) {layerSetup [{branchFilters [{boneName=upperarm_l}, {boneName=upperarm_r}]}]}
    bAlphaBoolEnabled:bool=True
   ) -> (Pose) {
    alphaInputType=Bool
    alphaBoolBlend {blendOption=Linear}
    alphaScaleBiasClamp {bInterpResult=true}
    controlRigAssetReference {blueprintRigClass=&/F_MED_RedOasisApricot_FaceAcc2_CtrlRig.F_MED_RedOasisApricot_FaceAcc2_CtrlRig_C}
   }
   Save cached pose 'preRBAN'(Pose=@n26) {cacheName=preRBAN}
  ]}
 EventGraph {nodes [
   Event BlueprintUpdateAnimation()
   > n1 = Sequence()
   n2 = Branch(execute=@n1.then_1, Condition=OR Boolean(A=Get bIsJumping(), B=Get bIsFalling()))
   > Set NewJumping(NewJumping:bool=true) -> (Output_Get:bool)
   Set NewJumping(execute=@n2.else, NewJumping=false)
   n3 = Gravity Override State Function DEPRECATED(
    execute=@n1.then_0
    CharacterPartAnimInstance=Self-Reference()
    GameTimeInSeconds=Get Game Time in Seconds() -> (Game Time in Seconds:double)
    GravityOverrideParams=Get Gravity Override Params()
   )
   > Set Gravity Override(Gravity Override=@n3)
   Event OnCharacterCustomizationLoaded()
   > n4 = Cast To PlayerPawn_Athena(Object=TryGetPawnOwner())
   > Set Body_Mesh(Body_Mesh=GetSkeletalMeshForPartType(self=@n4))
  ]}
]
variables [
 Body Mesh {type=SkeletalMeshComponent}
 Body_Mesh {type=SkeletalMeshComponent}
 Character Part Anim Instance {type=CustomCharacterPartAnimInstance, default=null}
 Game Time in Seconds {type=number}
 Gravity Override {type=Vector}
 Gravity Override Params {type=GravityOverrideParamsStruct, default {skydive_Parachuting {gravityOverride=<0 -800 -800>}}}
 HeadBack {type=number}
 IsADSAnim {type=boolean}
 NewJumping {type=boolean}
]
